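# Physics asset for the TestMelvin character in a Meshcapade test folder under the SpruceBark_M skin's menu animations.
# Eleven single-capsule bodies form a tree from a large root capsule through spine_01 and spine_05 to the arms, with both thighs on root.
# Ten constraints share uniform 45-degree swing and twist limits.
# Coverage is coarse and asymmetric: hands join upperarms directly, legs end at calf_r and foot_l, and thigh capsules are very thin.
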PhysicsAsset /BRCosmetics/Animation/Game/MainPlayer/Menu/BR/CustomSkins/_s42/SpruceBark_M/Meshcapade_Test/AS_TestMelvin_PhysicsAsset
bodies {
 calf_r [{shapeType=Capsule, length=25.1729, radius=6.6599, center=<25.3849 -1.7896 0.8639>, rotation=<rot 0 72.2075 -87.4677>}]
 foot_l [{shapeType=Capsule, length=10.0858, radius=5.6669, center=<-4.452 10.0645 -1.2862>, rotation=<rot 0 -163.9141 -69.0036>}]
 hand_l [{shapeType=Capsule, length=8.2386, radius=5.4422, center=<7.7885 -1.7579 -0.6855>, rotation=<rot 0 80.3407 -84.8268>}]
 hand_r [{shapeType=Capsule, length=8.2491, radius=5.4749, center=<-7.7754 1.7361 0.6533>, rotation=<rot 0 78.8716 -85.7046>}]
 root [{shapeType=Capsule, length=57.6378, radius=30.6484, center=<-17.0854 13.1401 52.1932>, rotation=<rot -2.8253 -90 -6.0262>}]
 spine_01 [{shapeType=Capsule, length=21.6317, radius=2.1735, center=<20.9583 -4.0684 0>, rotation=<rot 0 -105.4721 -89.9998>}]
 spine_05 [{shapeType=Capsule, length=19.8946, radius=11.6647, center=<11.4701 1.0532 -0.0001>, rotation=<rot 0 -90 -0.0005>}]
 thigh_l [{shapeType=Capsule, length=40.9654, radius=0.7866, center=<-40.5552 -0.2037 0>, rotation=<rot 0 -89.5761 -90>}]
 thigh_r [{shapeType=Capsule, length=20.9553, radius=0.5737, center=<20.7479 0 0>, rotation=<rot 0 -90 -90>}]
 upperarm_l [{shapeType=Capsule, length=24.541, radius=4.5998, center=<24.0463 4.4687 0>, rotation=<rot 0 -69.9111 -90>}]
 upperarm_r [{shapeType=Capsule, length=24.541, radius=4.5998, center=<-24.0463 -4.4687 0>, rotation=<rot 0 -69.9111 -90>}]
}
constraints [
 {bone1Name=spine_01, bone2Name=root, swing1LimitDegrees=45, swing2LimitDegrees=45, twistLimitDegrees=45}
 {bone1Name=spine_05, bone2Name=spine_01, swing1LimitDegrees=45, swing2LimitDegrees=45, twistLimitDegrees=45}
 {bone1Name=upperarm_l, bone2Name=spine_05, swing1LimitDegrees=45, swing2LimitDegrees=45, twistLimitDegrees=45}
 {bone1Name=hand_l, bone2Name=upperarm_l, swing1LimitDegrees=45, swing2LimitDegrees=45, twistLimitDegrees=45}
 {bone1Name=upperarm_r, bone2Name=spine_05, swing1LimitDegrees=45, swing2LimitDegrees=45, twistLimitDegrees=45}
 {bone1Name=hand_r, bone2Name=upperarm_r, swing1LimitDegrees=45, swing2LimitDegrees=45, twistLimitDegrees=45}
 {bone1Name=thigh_r, bone2Name=root, swing1LimitDegrees=45, swing2LimitDegrees=45, twistLimitDegrees=45}
 {bone1Name=calf_r, bone2Name=thigh_r, swing1LimitDegrees=45, swing2LimitDegrees=45, twistLimitDegrees=45}
 {bone1Name=thigh_l, bone2Name=root, swing1LimitDegrees=45, swing2LimitDegrees=45, twistLimitDegrees=45}
 {bone1Name=foot_l, bone2Name=thigh_l, swing1LimitDegrees=45, swing2LimitDegrees=45, twistLimitDegrees=45}
]
